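; Parent actor for the AvacadoEater bird, a flying parasite that hunts players and wildlife and can latch onto their heads.
; A state machine with timer loops moves it between patrol detection, homing-projectile chase, wall attacks and pawn attachment.
; Attacks apply gameplay effects and swap heads; detaching grants immunity and pops it away.
; RepNotifies drive attachment and death, with gameplay cues, sound indicators, vocal effects, health bar and an optional lifetime timer.
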(event Custom|StopGenericLoopingFeedback
  (params (Gameplay%20Cue%20Tag :type Gameplay%20Tag%20Structure))
  ;[1 0.7087 0] Start/Stop Looping Gameplay Cue Feedbacks for States
  (Projectile|RemoveGameplayCue%28Looping%29Local self Gameplay%20Cue%20Tag (Ability|GameplayCue|MakeGameplayCueParameters 0.0)))

(event Custom|StartGenericLoopingFeedback
  (params (Gameplay%20Cue%20Tag :type Gameplay%20Tag%20Structure))
  ;[1 0.7087 0] Start/Stop Looping Gameplay Cue Feedbacks for States
  (Projectile|AddGameplayCue%28Looping%29Local self Gameplay%20Cue%20Tag (Ability|GameplayCue|MakeGameplayCueParameters :NormalizedMagnitude 0.0 :RawMagnitude 0.0 :Location (Transformation|GetActorLocation) :Normal "0,0,0" :Instigator self :EffectCauser self :SourceObject self :GameplayEffectLevel 1 :AbilityLevel 1 :bReplicateLocationWhenUsingMinimalRepProxy false)))

(event Custom|StopPatrolFeedback
  ;[1 0.7087 0] Start/Stop Looping Gameplay Cue Feedbacks for States
  ;;[1 1 1] Patrol
  (Projectile|RemoveGameplayCue%28Looping%29Local self (Variables|GCs|GetGC_PatrolState) (Ability|GameplayCue|MakeGameplayCueParameters 0.0))
  (if (Utilities|Time|IsValidTimerHandle (stale Custom|StartPatrolFeedback _returnvalue))
    (Utilities|Time|ClearandInvalidateTimerbyHandle (stale Custom|StartPatrolFeedback _returnvalue))))

(event Custom|StartPatrolFeedback
  ;[1 0.7087 0] Start/Stop Looping Gameplay Cue Feedbacks for States
  ;;[1 1 1] Patrol
  (Projectile|AddGameplayCue%28Looping%29Local self (Variables|GCs|GetGC_PatrolState) (Ability|GameplayCue|MakeGameplayCueParameters :NormalizedMagnitude 0.0 :RawMagnitude 0.0 :Location (Transformation|GetActorLocation) :Normal "0,0,0" :Instigator self :EffectCauser self :SourceObject self :GameplayEffectLevel 1 :AbilityLevel 1 :bReplicateLocationWhenUsingMinimalRepProxy false))
  (bind _returnvalue (Utilities|Time|SetTimerbyEvent (delegate Custom|PatrolAVIndicatorLoop) 1.0 true)))

(event EventTick
  (params (DeltaSeconds :type Float%20%28single-precision%29)))

(event Custom|PatrolAVIndicatorLoop
  ;[1 0.7087 0] Start/Stop Looping Gameplay Cue Feedbacks for States
  ;;[1 1 1] Patrol
  (Fortnite|ShowSoundIndicatorLocal :TrackedActor self :Location (Transformation|GetActorLocation) :MaxAudibleDistance (Variables|Parameters|GetAVIndicatorMaxDistance) :IndicatorType "Infected" :Instigator self :Tint "(R=0.000000,G=0.000000,B=0.000000,A=0.000000)" :OverrideIcon (Variables|Parameters|GetAudioVisual%20Indicator%20Icon) :AttenuationMaxRangeMultiplier 1.0 :AttenuationIntensityMultiplier 1.0))

(event Custom|StopChaseFeedback
  ;[1 0.7087 0] Start/Stop Looping Gameplay Cue Feedbacks for States
  ;;[1 1 1] Chase
  (Projectile|RemoveGameplayCue%28Looping%29Local self (Variables|GCs|GetGC_ChaseState) (Ability|GameplayCue|MakeGameplayCueParameters 0.0))
  (if (Utilities|Time|IsValidTimerHandle (stale Custom|StartChaseFeedback _returnvalue))
    (Utilities|Time|ClearandInvalidateTimerbyHandle (stale Custom|StartChaseFeedback _returnvalue))))

(event Custom|PopUpAway
  (params (InitialPop :type Boolean))
  (bind _fortprojectilemovecomp_chase (Variables|Targeting|GetFortProjectileMoveComp_Chase))
  (bind _fortprojectilemovecomp_chase_2 (Variables|Targeting|GetFortProjectileMoveComp_Chase))
  ;[1 1 1] POP up and away
  (Development|PrintString (Utilities|String|Append "Popping Up & Away" (Utilities|String|ToString(Vector) (Transformation|GetActorLocation))) (Variables|Parameters|GetDebug) (Variables|Parameters|GetDebug))
  (Movement|StayAboveWater)
  ;;[1 1 1] Turn off homing and then turn ON simulation; required  before we can give the projectile a "pop" velocity
  (Projectile|SetHomingStyle _fortprojectilemovecomp_chase "None" true :!class /Script/FortniteGame.FortProjectileMovementComponent)
  (Components|Movement|SetUpdatedComponent _fortprojectilemovecomp_chase (Variables|Transformation|GetRootComponent))
  (Projectile|SetVelocity (Variables|Targeting|GetFortProjectileMoveComp_Chase) (Math|Vector|SelectVector (Math|Vector|vector+vector (* (Transformation|GetActorForwardVector) (Math|Random|RandomFloatinRange -2e+02 2e+02)) (* (Transformation|GetActorRightVector) (Math|Random|RandomFloatinRange -2e+02 2e+02)) (* (Transformation|GetActorUpVector) (Math|Random|RandomFloatinRange 7.5e+02 8.5e+02))) (Math|Vector|vector+vector (* "1.000000, 0, 0" (Math|Random|RandomFloatinRange -1.5e+02 1.5e+02)) (* "0,1.000000, 0" (Math|Random|RandomFloatinRange -1.5e+02 1.5e+02)) (* "0, 0,1.000000" (Math|Random|RandomFloatinRange 5.5e+02 6.5e+02))) InitialPop))
  (bind (_direction _gravscale _maxspeed) (Timeline|intro%20anim PlayFromStart)
    (:Update
      (Variables|/Script/Engine.ProjectileMovementComponent|SetProjectileGravityScale _gravscale _fortprojectilemovecomp_chase_2)
      (Variables|/Script/Engine.ProjectileMovementComponent|SetMaxSpeed _maxspeed _fortprojectilemovecomp_chase_2))
    (:Finished
      (Variables|/Script/Engine.ProjectileMovementComponent|SetMaxSpeed (ScalableFloat|GetValueatLevel (Variables|Balance|GetChaseMaxSpeed)) (Variables|Targeting|GetFortProjectileMoveComp_Chase)))))

(event EventEndPlay
  (params (EndPlayReason :type EEndPlayReason%20Enum :default "Destroyed"))
  ;[1 1 1] End Play
  (CallFunction|UpdateVocalEffects 0) ; Remove vocal FX if player leaves match
)

(event Custom|StartChaseFeedback
  ;[1 0.7087 0] Start/Stop Looping Gameplay Cue Feedbacks for States
  ;;[1 1 1] Chase
  (Projectile|AddGameplayCue%28Looping%29Local self (Variables|GCs|GetGC_ChaseState) (Ability|GameplayCue|MakeGameplayCueParameters :NormalizedMagnitude 0.0 :RawMagnitude 0.0 :Location (Transformation|GetActorLocation) :Normal "0,0,0" :Instigator self :EffectCauser self :SourceObject self :GameplayEffectLevel 1 :AbilityLevel 1 :bReplicateLocationWhenUsingMinimalRepProxy false))
  (bind _returnvalue (Utilities|Time|SetTimerbyEvent (delegate Custom|ChaseAVIndicatorLoop) 0.6 true)))

(event Custom|ChaseAVIndicatorLoop
  ;[1 0.7087 0] Start/Stop Looping Gameplay Cue Feedbacks for States
  ;;[1 1 1] Chase
  (Fortnite|ShowSoundIndicatorLocal :TrackedActor self :Location (Transformation|GetActorLocation) :MaxAudibleDistance (Variables|Parameters|GetAVIndicatorMaxDistance) :IndicatorType "Infected" :Instigator self :Tint "(R=0.000000,G=0.000000,B=0.000000,A=0.000000)" :OverrideIcon (Variables|Parameters|GetAudioVisual%20Indicator%20Icon) :AttenuationMaxRangeMultiplier 1.0 :AttenuationIntensityMultiplier 1.0))

(event EventBeginPlay
  (bind _instantattachtarget (Variables|Parameters|GetInstantAttachTarget))
  ;[1 1 1] Begin Play
  (|CallParentFunction)
  (Class|FortAsyncActionCurrentPlaylistReady|PlaylistDataReadyAsync)
  (Variables|Behavior|SetState "NewEnumerator5")
  (EventDispatchers|BindEventtoOnDamaged (EventDispatchers|CreateEvent :!event Custom|DamageEvent) :!class /Script/FortniteGame.BuildingActor)
  (if (Networking|IsServer)
    (CallFunction|StartLifetimeTimer)
    ;;[1 1 1] for wildlife instant attachment
    (if (Targeting|IsValidTarget _instantattachtarget)
      (Utilities|FlowControl|DelayUntilNextTick)
      (Development|PrintString (Utilities|String|Append "Instant attachment" (Utilities|GetDisplayName _instantattachtarget)) (Variables|Parameters|GetDebug) (Variables|Parameters|GetDebug))
      (Targeting|SetChaseTarget _instantattachtarget true (ScalableFloat|GetValueatLevel (Variables|Balance|GetChaseDistanceFromTargetToAttach)))
      (CallFunction|TryAttackingActor _instantattachtarget)
      (else
        ;;[1 1 1] Pop up out of egg & delay like we popped off player head
        (CallFunction|PopUpAway true)
        (Utilities|FlowControl|Delay (ScalableFloat|GetValueatLevel (Variables|Balance|GetPostHatchDelay)))
        (if (Targeting|IsValidTarget (Targeting|GetChaseTarget))
          (CallFunction|StartChaseLoop)
          (else
            (CallFunction|StartPatrolLoop)))))
    (else
      (CallFunction|StartUpdatingHealthbarWidget)
      (Health|BindEventtoOnBuildingHealthChanged (EventDispatchers|CreateEvent :!event Custom|OnBuildingHealthChanged_Event))
      ;;[1 1 1] Aim Assist
      (AimAssist|RegisterAimAssistTarget self (Variables|Collision|GetCapsuleComponent))
      (bind _asfort_pawn (Utilities|Casting|CastToFortPawn (Variables|/Script/Engine.PlayerState|GetPawnPrivate (Variables|/Script/Engine.Controller|GetPlayerState (Fortnite|GetFirstLocalFortPlayerController true)))))
      (FortPawn|SetAimAssistTargetOverride _asfort_pawn (Variables|Collision|GetCapsuleComponent))
      (GameplayCue|ExecuteGameplayCue%28Burst%29Local self (Variables|GCs|GetGC_Spawned) (Ability|GameplayCue|MakeGameplayCueParameters :NormalizedMagnitude 0.0 :RawMagnitude 0.0 :Location (Transformation|GetActorLocation) :Normal "0,0,0" :Instigator self :EffectCauser self :SourceObject self :GameplayEffectLevel 1 :AbilityLevel 1 :bReplicateLocationWhenUsingMinimalRepProxy false)))))

(event Custom|StopPatrolLoop
  ;[1 1 1] BEHAVIOR: PATROL LOGIC
  (Development|PrintString "Ending Patrol" (Variables|Parameters|GetDebug) (Variables|Parameters|GetDebug))
  (CallFunction|StopPatrolFeedback)
  (if (Utilities|Time|IsValidTimerHandle (stale Custom|StartPatrolLoop _returnvalue))
    (Utilities|Time|ClearandInvalidateTimerbyHandle (stale Custom|StartPatrolLoop _returnvalue)))
  (if (Utilities|Time|IsValidTimerHandle (stale Custom|StartPatrolLoop _returnvalue_2))
    (Utilities|Time|ClearandInvalidateTimerbyHandle (stale Custom|StartPatrolLoop _returnvalue_2))))

(event Custom|StartPatrolLoop
  (bind _fortprojectilemovecomp_chase (Variables|Targeting|GetFortProjectileMoveComp_Chase))
  ;[1 1 1] BEHAVIOR: PATROL LOGIC
  (Movement|StayAboveWater)
  (switch Utilities|FlowControl|Switch|SwitchonAvacadoEaterBird_States_Enum (Variables|Behavior|GetState)
    (:NewEnumerator1
      _)
    (:NewEnumerator7
      _)
    (:NewEnumerator3
      _)
    (:NewEnumerator5
      _))
  (Variables|Behavior|SetState "NewEnumerator0")
  (CallFunction|StopChaseLoop)
  (Development|PrintString "Starting Patrol" (Variables|Parameters|GetDebug) (Variables|Parameters|GetDebug))
  (CallFunction|StartPatrolFeedback)
  (bind _returnvalue (Utilities|Time|SetTimerbyEvent (delegate Custom|PatrolDetectionLoop) (ScalableFloat|GetValueatLevel (Variables|Balance|GetTargetDetectionRate)) true))
  (Projectile|SetHomingStyle (Variables|Targeting|GetFortProjectileMoveComp_Chase) "None" false :!class /Script/FortniteGame.FortProjectileMovementComponent)
  (Variables|/Script/Engine.ProjectileMovementComponent|SetProjectileGravityScale 0.0 _fortprojectilemovecomp_chase)
  (Projectile|SetVelocity _fortprojectilemovecomp_chase (* (Transformation|GetActorForwardVector) 1e+02))
  (bind _returnvalue_2 (Utilities|Time|SetTimerbyEvent (delegate Custom|PatrolMovementLoop) (ScalableFloat|GetValueatLevel (Variables|Balance|GetPatrolDirectionUpdateRate)) true false 0.0 (* (ScalableFloat|GetValueatLevel (Variables|Balance|GetPatrolDirectionUpdateRate)) 0.25)))
  (Utilities|FlowControl|Delay 1.55)
  (Targeting|RandomNewPatrolDestination))

(event Custom|DamageEvent
  (params (DamagedActor :type Actor%20Object%20Reference) (Damage :type Float%20%28single-precision%29) (InstigatedBy :type Controller%20Object%20Reference) (DamageCauser :type Actor%20Object%20Reference) (HitLocation :type Vector) (FHitComponent :type Primitive%20Component%20Object%20Reference) (BoneName :type Name) (Momentum :type Vector))
  ;[1 1 1] Damage: parasite takes damage
  (if (Utilities|NotEqual(Object) DamageCauser self)
    (Development|PrintString (Utilities|String|Append "Parasite Damaged for:  " (Utilities|String|ToString(Float) Damage) "   Remaining Health:  " (Utilities|String|ToString(Float) (Health|GetHealth))) (Variables|Parameters|GetDebug) (Variables|Parameters|GetDebug))
    (Variables|Behavior|SetdamagedRecently? true)
    ;;[1 0.7762 0] damage feedback
    (Ability|ExecuteGameplayCueWithParamsReplicated (Variables|GCs|GetGC_Damaged) (Ability|GameplayCue|MakeGameplayCueParameters :NormalizedMagnitude 0.0 :RawMagnitude 0.0 :Location HitLocation :Normal "0,0,0" :EffectCauser DamageCauser :GameplayEffectLevel 1 :AbilityLevel 1 :bReplicateLocationWhenUsingMinimalRepProxy false))
    (Utilities|Time|SetTimerbyEvent (delegate Custom|clearDamagedRecentlyDelay) 2.0)))

(event Custom|PatrolMovementLoop
  ;[1 1 1] BEHAVIOR: PATROL LOGIC
  (Targeting|RandomNewPatrolDestination))

(event Custom|PatrolDetectionLoop
  ;[1 1 1] BEHAVIOR: PATROL LOGIC
  (Targeting|LookforNewTarget) ; If a new target is found OnTargetAcquired will be called
)

(event Custom|clearDamagedRecentlyDelay
  ;[1 1 1] Damage: parasite takes damage
  (Variables|Behavior|SetdamagedRecently? false))

(event Custom|HideAndKill
  ;[1 1 1] Hide&Kill
  (Variables|Behavior|SetState "NewEnumerator4")
  (Ability|ExecuteGameplayCueReplicated (Variables|GCs|GetGC_Destroyed))
  (CallFunction|StopChaseLoop)
  (CallFunction|StopPatrolLoop)
  (Game|Components|ProjectileMovement|StopSimulating (Variables|Targeting|GetFortProjectileMoveComp_Chase) (Collision|MakeHitResult false))
  (Collision|SetActorEnableCollision false)
  (if (Networking|IsServer)
    (CallFunction|DetachfromHost)
    (Actor|SetLifeSpan (Math|Random|RandomFloatinRange 1.0 1.1))
    (else
      (CallFunction|NPCSetupHealthBaronDeath)
      (Components|Activation|SetActive (Variables|Default|GetFX_Niagara) false false)
      (Rendering|SetActorHiddenInGame true))))

(event Targeting|EventOnTargetAcquired
  ;[1 1 1] BEHAVIOR: PATROL LOGIC
  (CallFunction|StartChaseLoop))

(event EventOnDeathServer
  (params (Damage :type Float%20%28single-precision%29) (DamageTags :type Gameplay%20Tag%20Container%20Structure%20%28by%20ref%29) (Momentum :type Vector :default "0, 0, 0") (HitInfo :type Hit%20Result%20Structure%20%28by%20ref%29) (InstigatedBy :type Controller%20Object%20Reference) (DamageCauser :type Actor%20Object%20Reference) (EffectContext :type Gameplay%20Effect%20Context%20Handle%20Structure))
  ;[1 1 1] ON DEATH
  ;;[1 0.75 0] This is required to avoid running feedback and cleanup twice because this event is called both when HP reaches 0 AND when the actor is destroyed outright;
  (if (not (Variables|Behavior|GetIsDying?))
    (Variables|Behavior|SetIsDying? true)
    (Networking|ForceNetUpdate)))

(event Custom|StopChaseLoop
  ;[1 1 1] BEHAVIOR CHASE LOGIC
  (Development|PrintString "Ending Chase" (Variables|Parameters|GetDebug) (Variables|Parameters|GetDebug))
  (CallFunction|StopChaseFeedback)
  (if (Utilities|Time|IsValidTimerHandle (stale Custom|StartChaseLoop _returnvalue))
    (Utilities|Time|ClearandInvalidateTimerbyHandle (stale Custom|StartChaseLoop _returnvalue))))

(event Custom|StopLifetimeTimer
  ;[1 1 1] LIFETIME; plays feedback before triggering death
  (if (ScalableFloat|GetValueasBool (Variables|Balance|GetUseLifetimeTimer?))
    (Development|PrintString "Stopping Lifetimer" (Variables|Parameters|GetDebug) (Variables|Parameters|GetDebug))
    (if (Utilities|Time|IsValidTimerHandle (stale Custom|StartLifetimeTimer _returnvalue))
      (Utilities|Time|ClearandInvalidateTimerbyHandle (stale Custom|StartLifetimeTimer _returnvalue)))))

(event Custom|StartChaseLoop
  ;[1 1 1] BEHAVIOR CHASE LOGIC
  (switch Utilities|FlowControl|Switch|SwitchonAvacadoEaterBird_States_Enum (Variables|Behavior|GetState)
    (:NewEnumerator0
      _)
    (:NewEnumerator7
      _)
    (:NewEnumerator3
      _)
    (:NewEnumerator5
      _))
  (Variables|Behavior|SetState "NewEnumerator1")
  (CallFunction|StopPatrolLoop)
  (CallFunction|SetNewHomingTarget)
  (Development|PrintString "Starting Chase" (Variables|Parameters|GetDebug) (Variables|Parameters|GetDebug))
  (CallFunction|StartChaseFeedback)
  (bind _returnvalue (Utilities|Time|SetTimerbyEvent (delegate Custom|ChaseLoop) (ScalableFloat|GetValueatLevel (Variables|Balance|GetChaseRevalidateTargetUpdateRate)) true)))

(event Custom|StartLifetimeTimer
  ;[1 1 1] LIFETIME; plays feedback before triggering death
  (if (ScalableFloat|GetValueasBool (Variables|Balance|GetUseLifetimeTimer?))
    (Development|PrintString "Starting Lifetimer" (Variables|Parameters|GetDebug) (Variables|Parameters|GetDebug))
    (bind _returnvalue (Utilities|Time|SetTimerbyEvent (delegate Custom|LifetimerExpired) (ScalableFloat|GetValueatLevel (Variables|Balance|GetLifetimeDurationWhenUnattached))))))

(event Custom|LifetimerExpired
  ;[1 1 1] LIFETIME; plays feedback before triggering death
  (Development|PrintString "Lifetime Expired; destroying" (Variables|Parameters|GetDebug) (Variables|Parameters|GetDebug))
  (Variables|Behavior|GetREP_Attached
    (:then
      (CallFunction|DetachfromHost))
    (:else
      _))
  (CallFunction|StopPatrolLoop)
  (CallFunction|StopChaseLoop)
  (Variables|Behavior|SetState "NewEnumerator6")
  (Projectile|SetHomingStyle (Variables|Targeting|GetFortProjectileMoveComp_Chase) "None" false :!class /Script/FortniteGame.FortProjectileMovementComponent)
  (Game|Components|ProjectileMovement|StopSimulating (Variables|Targeting|GetFortProjectileMoveComp_Chase) (Collision|MakeHitResult false))
  (Ability|ExecuteGameplayCueReplicated (Variables|GCs|GetGC_LifetimeExpired))
  (Utilities|Time|SetTimerbyEvent (delegate Custom|lifetime%20feedback%20finished%20timer) (ScalableFloat|GetValueatLevel (Variables|Balance|GetLifetimeExpiredDestroyDelay))))

(event Custom|lifetime%20feedback%20finished%20timer
  ;[1 1 1] LIFETIME; plays feedback before triggering death
  (Variables|Behavior|SetIsDying? true)
  (Networking|ForceNetUpdate))

(event Custom|ChaseLoop
  ;[1 1 1] BEHAVIOR CHASE LOGIC
  (if (Targeting|IsValidTarget (Targeting|GetChaseTarget))
    ;;[1 1 1] Check for Target being out of Tracking Range - go back to patrol if so
    (if (> (Class|AbilityFunctions|GetDistanceBetweenActors (Targeting|GetChaseTarget) self) (ScalableFloat|GetValueatLevel (Variables|Balance|GetTargetingRangeMax)))
      (Development|PrintString "Target out of chase range, going back to patrol" (Variables|Parameters|GetDebug) (Variables|Parameters|GetDebug))
      (CallFunction|StartPatrolLoop))
    (else
      (CallFunction|StartPatrolLoop))))

(event OnChaseTargetReached(FortProjectileMoveComp_Chase)
  ;[1 1 1] BEHAVIOR: Impacts, Bounces, and Triggering Attacks
  ;;[1 1 1] CHASE TARGET REACHED; trigger condition for starting attacks
  (Development|PrintString "Reached chase target")
  (CallFunction|TryAttackingActor (Targeting|GetChaseTarget)))

(event Custom|tryAttackingActor
  (params (Target%20Actor :type Actor%20Object%20Reference))
  ;[1 1 1] BEHAVIOR: Impacts, Bounces, and Triggering Attacks
  ;;[1 1 1] Initiating Attacks on Characters
  (switch Utilities|FlowControl|Switch|SwitchonAvacadoEaterBird_States_Enum (Variables|Behavior|GetState)
    (:NewEnumerator1
      _)
    (:NewEnumerator5
      _))
  (bind _ascharacter (Utilities|Casting|CastToCharacter Target%20Actor)
    (:then
      (if (Targeting|IsValidTarget _ascharacter)
        (Development|PrintString (Utilities|String|Append "Attacking a character" (Utilities|GetDisplayName _ascharacter)) (Variables|Parameters|GetDebug) (Variables|Parameters|GetDebug))
        (CallFunction|AttackPawnSequence _ascharacter)
        (else
          _)))
    (:CastFailed
      (Development|PrintString "not a character; attack failed" (Variables|Parameters|GetDebug) (Variables|Parameters|GetDebug))))
  (Development|PrintString "Target is invalid; bouncing" (Variables|Parameters|GetDebug) (Variables|Parameters|GetDebug))
  (CallFunction|StartPatrolLoop))

(event Custom|StartUpdatingHealthbarWidget
  ;[1 0.2678 0] HP Bar
  (bind _statuswidget (Variables|Default|GetStatusWidget)
    (:then
      (bind _asnpcstatus_widget (Utilities|Casting|CastToNPCStatusWidget (UserInterface|GetUserWidgetObject _statuswidget)))
      (Utilities|FlowControl|Sequence
        (:then_0
          (Components|Tick|SetComponentTickEnabled (Variables|Default|GetStatusWidget) (Widget|IsVisible _asnpcstatus_widget)))
        (:then_1
          (Default|BindEventtoVisibilityChanged _asnpcstatus_widget (delegate Custom|HealthBarWidgetVisibilityChanged) :!class /Game/Athena/HUD/Actors/NPCStatusWidget.NPCStatusWidget_C))))
    (:else)))

(event Custom|HealthBarWidgetVisibilityChanged
  (params (bVisible :type Boolean))
  ;[1 0.2678 0] HP Bar
  (Components|Tick|SetComponentTickEnabled (Variables|Default|GetStatusWidget) bVisible))

(event Collision|EventHit
  (params (MyComp :type Primitive%20Component%20Object%20Reference) (Other :type Actor%20Object%20Reference) (OtherComp :type Primitive%20Component%20Object%20Reference) (bSelfMoved :type Boolean) (HitLocation :type Vector :default "0, 0, 0") (HitNormal :type Vector :default "0, 0, 0") (NormalImpulse :type Vector :default "0, 0, 0") (Hit :type Hit%20Result%20Structure%20%28by%20ref%29))
  (bind (_bblockinghit _binitialoverlap _time _distance _location _impactpoint _normal _impactnormal _physmat _hitactor) (Collision|BreakHitResult Hit))
  ;[1 1 1] BEHAVIOR: Impacts, Bounces, and Triggering Attacks
  ;;[1 1 1] ON HIT EVENT: Used for hitting walls. Note currently the parasite is intentionally set up to NOT collide with pawns (annoying to block movement) so they trigger attack on overlap while walls still require on hit
  (switch Utilities|FlowControl|Switch|SwitchonAvacadoEaterBird_States_Enum (Variables|Behavior|GetState)
    (:NewEnumerator1
      (if (Networking|IsServer)
        ;;;[1 0.7164 0] Wall detection - how we attack walls; needs improvement! parasite behavior is often better if this is entirely disconnected as the parasite can go around walls & through doors if this is disabled
        (bind _asbuilding_static_mesh (Utilities|Casting|CastToBuildingSMActor _hitactor))
        (Development|PrintString "attacking a BuildingSMActor" (Variables|Parameters|GetDebug) (Variables|Parameters|GetDebug))
        (CallFunction|AttackBuildingActorSequence _asbuilding_static_mesh)))))

(event Custom|StopUpdatingHealthBarWidget
  ;[1 0.2678 0] HP Bar
  (bind _statuswidget (Variables|Default|GetStatusWidget)
    (:then
      (bind _asnpcstatus_widget (Utilities|Casting|CastToNPCStatusWidget (UserInterface|GetUserWidgetObject _statuswidget)))
      (Default|UnbindEventfromVisibilityChanged _asnpcstatus_widget (delegate Custom|HealthBarWidgetVisibilityChanged) :!class /Game/Athena/HUD/Actors/NPCStatusWidget.NPCStatusWidget_C))
    (:else)))

(event Custom|Attack%20Pawn%20Sequence
  (params (Attack%20Target :type Actor%20Object%20Reference))
  (bind _attachtoheadduration (Variables|Balance|GetAttachToHeadDuration))
  (bind _returnvalue (ScalableFloat|GetValueatLevel _attachtoheadduration))
  (bind _returnvalue_2 (Ability|GetFortAbilitySystemComponent Attack%20Target))
  ;[1 1 1] ATTACK PLAYER/WILDLIFE SEQUENCE - super long sequence with optional functionality for attaching to head on attack
  (Utilities|IsValid (Ability|GetFortAbilitySystemComponent Attack%20Target)
    (:Is%20Valid
      (CallFunction|StopChaseLoop)
      (CallFunction|StopPatrolLoop)
      (if (Targeting|IsValidTarget (Targeting|GetChaseTarget))
        (Variables|Behavior|SetState "NewEnumerator2")
        (Development|PrintString (Utilities|String|Append "Attaching to PAWN:  " (Utilities|GetDisplayName Attack%20Target)) (Variables|Parameters|GetDebug) (Variables|Parameters|GetDebug))
        ;;[1 1 1] SERVER: Attack Gameplay Effect applied to victim
        (bind _returnvalue_3 (GameplayTags|HasMatchingGameplayTag(Message) Attack%20Target "(TagName=\"Pawn.Athena.NPC.Wildlife\")"))
        (if _returnvalue_3
          ;;[0.2338 1 0] WILDLIFE: apply wildlife GE
          (bind _returnvalue_4 (GameplayEffects|FortApplyGameplayEffectToTarget :self (Ability|GetFortAbilitySystemComponent self) :GameplayEffectClass (Variables|GEs|GetGE_ApplyToWildlifeOnAttach) :TargetOfEffect _returnvalue_2 :Level 0.0))
          ;;;[1 1 1] used to remove the duration GE if this dies before GE finishes
          (Variables|Behavior|SetGE%20Applied%20on%20Target _returnvalue_4)
          (else
            (bind _returnvalue_5 (GameplayEffects|FortApplyGameplayEffectToTarget :self (Ability|GetFortAbilitySystemComponent self) :GameplayEffectClass (Variables|GEs|GetGE_ApplyToPlayersOnAttach) :TargetOfEffect _returnvalue_2 :Level 0.0))
            ;;;[1 1 1] used to remove the duration GE if this dies before GE finishes
            (Variables|Behavior|SetGE%20Applied%20on%20Target _returnvalue_5)))
        (if (ScalableFloat|GetValueasBool (Variables|Balance|GetAttachtoHeadOnAttack?))
          ;;[1 1 1] SERVER: Attach to target's head and then delay a moment
          (bind _ascharacter (Utilities|Casting|CastToCharacter Attack%20Target)
            (:then
              (Variables|Behavior|SetREP_Attached _ascharacter)
              (Networking|ForceNetUpdate)
              (CallFunction|SwapHead Attack%20Target)
              (CallFunction|StopLifetimeTimer))
            (:CastFailed
              _)))
        ;;[1 1 1] SERVER: Apply Gameplay Effects to Parasite
        (if (ScalableFloat|GetValueasBool (Variables|Balance|GetHealOnHit?))
          (bind _returnvalue_6 (GameplayEffects|ApplyGameplayEffectToSelf (Ability|GetFortAbilitySystemComponent self) (Variables|GEs|GetGE_ApplyToSelfOnAttach) 0.0 (GameplayEffects|MakeEffectContext (Ability|GetFortAbilitySystemComponent self))))
          (Variables|Behavior|SetGE%20Applied%20to%20Slef%20On%20Attach _returnvalue_6))
        ;;[1 1 1] SERVER & CLIENT: Delay while on head before Detatch from Head
        (if (> _returnvalue 0.0)
          (Utilities|Time|SetTimerbyEvent (delegate Custom|AttackSequenceDelay2) _returnvalue)
          (else
            (Development|PrintString "Attaching indefinitely to host" (Variables|Parameters|GetDebug) (Variables|Parameters|GetDebug))))
        (else
          _)))
    (:Is%20Not%20Valid
      _))
  ;;[1 1 1] Early Outs and Edge Cases
  (Development|PrintString "Early Out of Attachment" (Variables|Parameters|GetDebug) (Variables|Parameters|GetDebug))
  (CallFunction|StartPatrolLoop))

(event Custom|DetachFromHost
  (bind _rep_attached (Variables|Behavior|GetREP_Attached))
  (bind _returnvalue (Ability|GetAbilitySystemComponent _rep_attached))
  ;[1 1 1] ATTACK PLAYER/WILDLIFE SEQUENCE - super long sequence with optional functionality for attaching to head on attack
  ;;[1 1 1] Detatch
  (Utilities|IsValid _rep_attached
    (:Is%20Valid
      (GameplayEffects|RemoveActiveGameplayEffect _returnvalue (Variables|Behavior|GetGE%20Applied%20on%20Target) -1)
      (GameplayEffects|FortApplyGameplayEffectToTarget :self (Ability|GetFortAbilitySystemComponent self) :GameplayEffectClass (Variables|GEs|GetGE_ApplyImmunityTag) :TargetOfEffect _returnvalue :Level 0.0)
      ;;[1 1 1] POST-DETACH EFFECTS ON TARGET: Remove the Applied Effect and Apply post-detatchment immunity to the victim
      (CallFunction|RestoreHead (Variables|Behavior|GetREP_Attached))
      (Variables|Behavior|SetREP_Attached 0)
      (Networking|ForceNetUpdate))
    (:Is%20Not%20Valid
      (Development|PrintString "Error: trying to detatch while not attached" (Variables|Parameters|GetDebug) (Variables|Parameters|GetDebug))))
  (switch Utilities|FlowControl|Switch|SwitchonAvacadoEaterBird_States_Enum (Variables|Behavior|GetState)
    (:NewEnumerator2
      (Variables|Behavior|SetState "NewEnumerator3")
      (GameplayEffects|RemoveActiveGameplayEffect (Ability|GetFortAbilitySystemComponent self) (Variables|Behavior|GetGE%20Applied%20to%20Slef%20On%20Attach) -1) ; Remove the self-heal and "unblock" the passive health drain
      (CallFunction|StartLifetimeTimer)
      (CallFunction|PopUpAway false)
      (Ability|ExecuteGameplayCueReplicated (Variables|GCs|GetGC_DetachFromHost))
      ;;[1 1 1] SERVER: Resume Chase after a "long" delay that gives player a chance to shoot
      (Development|PrintString "Post-attach/attack delay")
      ;;;[1 1 1] Delay Before resume
      (Utilities|Time|SetTimerbyEvent (delegate Custom|AttackSequenceDelay4) (ScalableFloat|GetValueatLevel (Variables|Balance|GetPostAttackDelay))))
    (:NewEnumerator4
      _)
    (:NewEnumerator6
      _))
  (Development|PrintString "Dying"))

(event Custom|NPC%20SetupHealthBarOnDeath
  (bind _healthbarhidetimeafterdeath (Variables|Balance|GetHealthBarHideTimeAfterDeath))
  (bind _returnvalue (ScalableFloat|GetValueatLevel _healthbarhidetimeafterdeath))
  ;[1 0.2678 0] HP Bar
  (if (> _returnvalue 0.0)
    (bind _statuswidget (Variables|Default|GetStatusWidget)
      (:then
        (bind _asnpcstatus_widget (Utilities|Casting|CastToNPCStatusWidget (UserInterface|GetUserWidgetObject _statuswidget)))
        (Class|NPCStatusWidget|EventUpdateStatus _asnpcstatus_widget 0.0 0.0 false)
        (Utilities|FlowControl|Delay _returnvalue))
      (:else)))
  (bind _statuswidget_2 (Variables|Default|GetStatusWidget)
    (:then
      (Development|SetHiddeninGame _statuswidget_2 true false))
    (:else)))

(event Custom|Bounce
  ;[1 1 1] BEHAVIOR: Impacts, Bounces, and Triggering Attacks
  ;;[1 1 1] Bounce
  (Development|PrintString "Bouncing off target" (Variables|Parameters|GetDebug) (Variables|Parameters|GetDebug))
  (Projectile|SetHomingStyle (Variables|Targeting|GetFortProjectileMoveComp_Chase) "None" true :!class /Script/FortniteGame.FortProjectileMovementComponent)
  (Projectile|SetVelocity (Variables|Targeting|GetFortProjectileMoveComp_Chase) (* (Math|Vector|vector+vector (* (Transformation|GetActorForwardVector) (Math|Random|RandomFloatinRange -1.5 -1.6)) (* (Transformation|GetActorRightVector) (Math|Random|RandomFloatinRange 0.0)) (* (Transformation|GetActorUpVector) (Math|Random|RandomFloatinRange 1.5 1.6))) 5e+02))
  (if (Networking|IsServer)
    (Ability|ExecuteGameplayCueReplicated (Variables|GCs|GetGC_Bounce))
    (Utilities|Time|SetTimerbyEvent (delegate Custom|bouncedelay) 0.25)))

(event Custom|AttackSequenceDelay2
  ;[1 1 1] ATTACK PLAYER/WILDLIFE SEQUENCE - super long sequence with optional functionality for attaching to head on attack
  ;;[1 1 1] SERVER & CLIENT: Delay while on head before Detatch from Head
  (CallFunction|DetachfromHost))

(event Custom|AttackSequenceDelay4
  ;[1 1 1] ATTACK PLAYER/WILDLIFE SEQUENCE - super long sequence with optional functionality for attaching to head on attack
  ;;[1 1 1] SERVER: Resume Chase after a "long" delay that gives player a chance to shoot
  (Development|PrintString "Post-attack delay finished; resuming chase")
  (Variables|Behavior|SetAttachedCached 0)
  (Variables|Default|SetAttachedPlayerControllerCached 0)
  (switch Utilities|FlowControl|Switch|SwitchonAvacadoEaterBird_States_Enum (Variables|Behavior|GetState)
    (:NewEnumerator2
      _)
    (:NewEnumerator3
      _))
  (if (not (Variables|Behavior|GetIsDying?))
    (CallFunction|StartPatrolLoop)))

(event Custom|bouncedelay
  ;[1 1 1] BEHAVIOR: Impacts, Bounces, and Triggering Attacks
  ;;[1 1 1] Bounce
  (CallFunction|StartChaseLoop))

(event Custom|OnBuildingHealthChanged_Event
  ;[1 0.2678 0] HP Bar
  (if (Math|Boolean|ANDBoolean (not (Variables|Behavior|GetIsDying?)) (Variables|Parameters|GetUseNPCHealthBar) (< (Health|GetHealthPercent) 1.0))
    (bind _array_element (Utilities|Array|ForEachLoopwithBreak (Fortnite|GetLocalFortPlayerControllers true) :!label _label)
      (:LoopBody
        (if (< (Class|AbilityFunctions|GetDistanceBetweenActors (Pawn|GetControlledPawn _array_element) self) (ScalableFloat|GetValueatLevel (Variables|Balance|GetHealthBarMaxDisplayRange)))
          (bind _statuswidget (Variables|Default|GetStatusWidget)
            (:then
              (bind _asnpcstatus_widget (Utilities|Casting|CastToNPCStatusWidget (UserInterface|GetUserWidgetObject _statuswidget)))
              (Class|NPCStatusWidget|EventUpdateStatus _asnpcstatus_widget (Health|GetHealthPercent) 0.0 false :!pinned true) ; Show Health Bar
              (goto _label Break))
            (:else)))))))

(event Custom|Attack%20Building%20Actor%20Sequence
  (params (Attack%20Target :type Actor%20Object%20Reference))
  ;[1 1 1] ATTACK BUILDING SEQUENCE
  (Utilities|IsValid Attack%20Target
    (:Is%20Valid
      (Variables|Behavior|SetState "NewEnumerator7")
      (Development|PrintString (Utilities|String|Append "Attacking Building Actor: " (Utilities|GetDisplayName Attack%20Target)) (Variables|Parameters|GetDebug) (Variables|Parameters|GetDebug))
      (CallFunction|StopChaseLoop)
      (CallFunction|StopPatrolLoop)
      (GameplayEffects|FortApplyGameplayEffectToTarget :self (Ability|GetFortAbilitySystemComponent self) :GameplayEffectClass (Variables|GEs|GetGE%20Apply%20to%20Buildings) :TargetOfEffect (Ability|GetFortAbilitySystemComponent Attack%20Target) :OptionalEffectCauser self :OptionalDamageSource self :Level 0.0)
      (CallFunction|AttackBuildingMulticastEffects)
      (Utilities|Time|SetTimerbyEvent (delegate Custom|Dlay%20while%20attacking) (ScalableFloat|GetValueatLevel (Variables|Balance|GetAttackWallDuration))))
    (:Is%20Not%20Valid
      (CallFunction|StartPatrolLoop))))

(event Custom|NPCEnableHealthBars
  ;[1 0.2678 0] HP Bar
  ;;[1 1 1] CHEATS: Mostly for capture team stuff to hide bars across the board.
  (Variables|Parameters|SetUseNPCHealthBar true))

(event Custom|postbuildingattackdlay
  ;[1 1 1] ATTACK BUILDING SEQUENCE
  ;;[1 1 1] Rotate actor to face chase target and then resume chase after a short delay
  (CallFunction|StartChaseLoop))

(event Custom|Dlay%20while%20attacking
  (bind (_returnvalue_roll _returnvalue_pitch _returnvalue_yaw) (Math|Rotator|BreakRotator (Math|Rotator|FindLookatRotation (Transformation|GetActorLocation) (Transformation|GetActorLocation (Targeting|GetChaseTarget)))))
  ;[1 1 1] ATTACK BUILDING SEQUENCE
  ;;[1 0 0.6912] old bounce hack
  (Transformation|Teleport (Math|Vector|vector+vector (* (Transformation|GetActorForwardVector) -2e+01) (Transformation|GetActorLocation) (* (Transformation|GetActorUpVector) 15.0)) (Transformation|GetActorRotation))
  (if (Targeting|IsValidTarget (Targeting|GetChaseTarget))
    ;;[1 1 1] Rotate actor to face chase target and then resume chase after a short delay
    (Transformation|SetActorRotation (Math|Rotator|MakeRotator 0.0 0.0 _returnvalue_yaw) true)
    (Utilities|Time|SetTimerbyEvent (delegate Custom|postbuildingattackdlay) (ScalableFloat|GetValueatLevel (Variables|Balance|GetAttackWallDuration)))
    (else
      (CallFunction|StartPatrolLoop))))

(event Custom|NPCDisableHealthBars
  ;[1 0.2678 0] HP Bar
  ;;[1 1 1] CHEATS: Mostly for capture team stuff to hide bars across the board.
  (Variables|Parameters|SetUseNPCHealthBar false)
  (bind _statuswidget (Variables|Default|GetStatusWidget)
    (:then
      (bind _asnpcstatus_widget (Utilities|Casting|CastToNPCStatusWidget (UserInterface|GetUserWidgetObject _statuswidget)))
      (Widget|SetVisibility _asnpcstatus_widget "Hidden"))
    (:else)))

(event Custom|AttackBuildingMulticastEffects
  ;[1 1 1] ATTACK BUILDING SEQUENCE
  ;;[1 0.8315 0] feedback & stop moving
  (Projectile|SetHomingStyle (Variables|Targeting|GetFortProjectileMoveComp_Chase) "None" false :!class /Script/FortniteGame.FortProjectileMovementComponent)
  (Projectile|SetVelocity (Variables|Targeting|GetFortProjectileMoveComp_Chase) (Math|Vector|MakeVector 0.0))
  (Fortnite|ShowSoundIndicatorLocal :TrackedActor self :Location (Transformation|GetActorLocation) :MaxAudibleDistance (Variables|Parameters|GetAVIndicatorMaxDistance) :IndicatorType "Infected" :Instigator self :Tint "(R=0.000000,G=0.000000,B=0.000000,A=0.000000)" :OverrideIcon (Variables|Parameters|GetAudioVisual%20Indicator%20Icon) :AttenuationMaxRangeMultiplier 1.0 :AttenuationIntensityMultiplier 1.0))

(event Custom|Stop%20Moving)

(fn GetTargetOwner
  (return 0))

(fn IsValidAutoFireTarget
  (params (Attacker :type Actor%20Object%20Reference))
  (return true))

(fn OnRep_IsDying?
  (if (Variables|Behavior|GetIsDying?)
    (Development|PrintString "RepNotify Dying")
    (CallFunction|HideandKill)))

(fn OnRep_REP_Attached
  (bind _capsulecomponent (Variables|Collision|GetCapsuleComponent))
  (bind _rep_attached (Variables|Behavior|GetREP_Attached))
  (bind _capsulecomponent_2 (Variables|Collision|GetCapsuleComponent))
  (bind _attachedcached (Variables|Behavior|GetAttachedCached))
  (bind _gc_attachedstate (Variables|GCs|GetGC_AttachedState))
  (Utilities|IsValid _rep_attached
    (:Is%20Valid
      ;[1 1 1] Replicated Multicast On Attach
      (Development|PrintString "attaching parasite on client" (Variables|Parameters|GetDebug) (Variables|Parameters|GetDebug))
      (Game|Components|ProjectileMovement|StopSimulating (Variables|Targeting|GetFortProjectileMoveComp_Chase) (Collision|MakeHitResult false false 0.0 0.0 "0, 0, 0" "0, 0, 0" "0,0,1" "0,0,1" 0 _rep_attached))
      ;;[1 1 1] Cache this on the client so that we have a valid ref when we come back into OnRep on detach
      (bind _output_get (Variables|Behavior|SetAttachedCached _rep_attached)) ; Cache this on the client
      (Variables|Default|SetAttachedPlayerControllerCached (Fortnite|GetFortPlayerControllerfromActor _output_get))
      (bind _returnvalue (GameplayTags|HasMatchingGameplayTag(Message) _rep_attached "(TagName=\"Pawn.Athena.NPC.Wildlife\")"))
      (Transformation|AttachActorToComponent (Variables|/Script/Engine.Character|GetMesh _rep_attached) (select _returnvalue (Variables|Targeting|GetTargetNpcSocketName) (Variables|Targeting|GetTargetPlayerSocketName)) "SnapToTarget" "SnapToTarget" "KeepWorld" false)
      (Transformation|SetActorRelativeRotation "0.000000,90.000000,-90.000000" false true)
      ;;[1 1 1] Specific Collisions Disabled
      (Collision|SetCollisionResponsetoChannel _capsulecomponent "ECC_Vehicle" "ECR_Ignore") ; Disable Vehicle collisions while attached; prevents crazy stuff when getting in vehicles
      (Collision|SetCollisionResponsetoChannel _capsulecomponent "ECC_GameTraceChannel1" "ECR_Ignore") ; Prevent problems with parasite blocking pickups
      (Collision|SetCollisionResponsetoChannel _capsulecomponent "ECC_PhysicsBody" "ECR_Ignore") ; Re-enable ability to ping parasite
      ;;[1 1 1] Disable parasite VFX particles
      (Components|Activation|SetActive (Variables|Default|GetFX_Niagara) false false)
      ;;[1 0.7106 0] Client Only
      (Networking|CanExecuteCosmeticEvents
        (:True
          (Utilities|FlowControl|Sequence
            (:then_0
              (CallFunction|UpdateVocalEffects "/Game/Athena/Items/Consumables/AvacadoEater/DSPEffects/AvacadoEater_EffectChain.AvacadoEater_EffectChain")
              (CallFunction|StartGenericLoopingFeedback _gc_attachedstate)
              (Fortnite|ShowSoundIndicatorLocal :TrackedActor self :Location (Transformation|GetActorLocation) :MaxAudibleDistance (Variables|Parameters|GetAVIndicatorMaxDistance) :IndicatorType "Infected" :Instigator self :Tint "(R=0.000000,G=0.000000,B=0.000000,A=0.000000)" :OverrideIcon (Variables|Parameters|GetAudioVisual%20Indicator%20Icon) :AttenuationMaxRangeMultiplier 1.0 :AttenuationIntensityMultiplier 1.0)
              (Rendering|SetVisibility (Variables|Default|GetSkeletalMesh) false false)
              ;;;[1 0 0.1184] Disable Aim Assist on attached parasites
              (AimAssist|UnregisterAimAssistTarget self)
              (bind _asfort_pawn (Utilities|Casting|CastToFortPawn (Variables|/Script/Engine.PlayerState|GetPawnPrivate (Variables|/Script/Engine.Controller|GetPlayerState (Fortnite|GetFirstLocalFortPlayerController true)))))
              (FortPawn|SetAimAssistTargetOverride _asfort_pawn 0))
            (:then_1
              ;;;[1 1 1] Tutorial
              (bind _asfort_player_state (Utilities|Casting|CastToFortPlayerState (Variables|/Script/Engine.Pawn|GetPlayerState (Variables|Behavior|GetREP_Attached))))
              (if (Pawn|IsLocallyControlled (PlayerState|GetCurrentPawn _asfort_player_state))
                (Ability|SendGameplayEventtoActor (PlayerState|GetCurrentPawn _asfort_player_state) (Variables|Parameters|GetTutorial%20Attached))))))))
    (:Is%20Not%20Valid
      ;[1 1 1] Replicated Multicast On Detatched
      (Development|PrintString "Parasite Detaching" (Variables|Parameters|GetDebug) (Variables|Parameters|GetDebug))
      (Transformation|DetachFromComponent (Variables|Transformation|GetRootComponent) "KeepWorld" "KeepWorld" "KeepWorld" true)
      ;;[1 1 1] Specific Collisions Reneabled
      (Collision|SetCollisionResponsetoChannel _capsulecomponent_2 "ECC_Vehicle" "ECR_Block") ; Reenable Vehicle collisions on detatch so you can now hit parasites with vehicles
      (Collision|SetCollisionResponsetoChannel _capsulecomponent_2 "ECC_GameTraceChannel1" "ECR_Block") ; Re-enable ability to ping parasite
      (Collision|SetCollisionResponsetoChannel _capsulecomponent_2 "ECC_PhysicsBody" "ECR_Block") ; Re-enable ability to ping parasite
      ;;[1 1 1] Reenable parasite VFX particles
      (Components|Activation|SetActive (Variables|Default|GetFX_Niagara) true true)
      ;;[1 0.6204 0] Client Only
      (Networking|CanExecuteCosmeticEvents
        (:True
          (Utilities|FlowControl|Sequence
            (:then_0
              (CallFunction|UpdateVocalEffects 0) ; This is handled inside REP_Attached_Previous
              (CallFunction|StopGenericLoopingFeedback _gc_attachedstate)
              (Rendering|SetVisibility (Variables|Default|GetSkeletalMesh) true false)
              ;;;[1 0 0.1184] Re-enable aim assist for detached parasites
              (AimAssist|RegisterAimAssistTarget self (Variables|Collision|GetCapsuleComponent))
              (bind _asfort_pawn_2 (Utilities|Casting|CastToFortPawn (Variables|/Script/Engine.PlayerState|GetPawnPrivate (Variables|/Script/Engine.Controller|GetPlayerState (Fortnite|GetFirstLocalFortPlayerController true)))))
              (FortPawn|SetAimAssistTargetOverride _asfort_pawn_2 (Variables|Collision|GetCapsuleComponent)))
            (:then_1
              ;;;[1 1 1] Tutorial
              (Utilities|IsValid _attachedcached
                (:Is%20Valid
                  (bind _asfort_player_state_2 (Utilities|Casting|CastToFortPlayerState (Variables|/Script/Engine.Pawn|GetPlayerState _attachedcached)))
                  (if (Pawn|IsLocallyControlled (PlayerState|GetCurrentPawn _asfort_player_state_2))
                    (Ability|SendGameplayEventtoActor (PlayerState|GetCurrentPawn _asfort_player_state_2) (Variables|Parameters|GetTutorial%20Detached))))))))))))

(fn RestoreHead
  (params (Headless%20Actor :type Actor%20Object%20Reference))
  (bind _gadgetitemheadswap (Variables|Animation|GetGadgetItemHeadSwap))
  (bind _attachedplayercontrollercached (Variables|Default|GetAttachedPlayerControllerCached))
  ;[1 1 1] Handle PlayerPawn
  (if (Utilities|IsValid _attachedplayercontrollercached :!class /Script/Engine.KismetSystemLibrary)
    (Inventory|RemoveItemFromPlayer :PlayerController _attachedplayercontrollercached :ItemDefinition (Variables|Animation|GetGadgetItemHeadSwap) :AmountToRemove -1 :bForceRemoval true :bSkipVerbRemoval false)
    (else
      ;[1 1 1] Handle AIBot
      (bind _asfort_player_pawn (Utilities|Casting|CastToFortPlayerPawn Headless%20Actor)
        (:then
          (bind _asfort_athena_aibot_controller (Utilities|Casting|CastToFortAthenaAIBotController (Pawn|GetController _asfort_player_pawn))
            (:then
              (bind _asfort_inventory_owner_interface (Utilities|Casting|CastToFortInventoryOwnerInterface _asfort_athena_aibot_controller)
                (:then
                  (Inventory|RemoveItemfromInventoryOwner _asfort_inventory_owner_interface _gadgetitemheadswap)
                  (return))
                (:CastFailed
                  _)))
            (:CastFailed
              _)))
        (:CastFailed
          ;[1 1 1] Handle AIPawn
          (bind _asfort_aipawn (Utilities|Casting|CastToFortAIPawn Headless%20Actor)
            (:then
              (bind _asfort_inventory_owner_interface_2 (Utilities|Casting|CastToFortInventoryOwnerInterface _asfort_aipawn)
                (:then
                  (Inventory|RemoveItemfromInventoryOwner _asfort_inventory_owner_interface_2 _gadgetitemheadswap)
                  (return))
                (:CastFailed
                  _)))
            (:CastFailed
              _))))
      (return)))
  (return))

(fn SetNewHomingTarget
  (bind _returnvalue (Targeting|GetChaseTarget))
  (bind _fortprojectilemovecomp_chase (Variables|Targeting|GetFortProjectileMoveComp_Chase))
  (Utilities|IsValid _returnvalue
    (:Is%20Valid
      (bind _returnvalue_2 (GameplayTags|HasMatchingGameplayTag(Message) _returnvalue "(TagName=\"Pawn.Athena.NPC.Wildlife\")"))
      (Variables|Targeting|SetTargetSocketName (select _returnvalue_2 (Variables|Targeting|GetTargetNpcSocketName) (Variables|Targeting|GetTargetPlayerSocketName)))
      (Projectile|SetHomingData _fortprojectilemovecomp_chase (Utilities|Struct|MakeProjectileHomingData "LockOn" (ScalableFloat|GetValueatLevel (Variables|Balance|GetChaseTurnSpeedMin)) (ScalableFloat|GetValueatLevel (Variables|Balance|GetChaseTurnSpeedMax)) (ScalableFloat|GetValueatLevel (Variables|Balance|GetChaseTurnSpeedRampTime)) (ScalableFloat|GetValueatLevel (Variables|Balance|GetTargetingRangeMax)) (Math|Random|RandomIntegerinRange 0 99) (Targeting|GetChaseTarget) 0 "0.000000,0.000000,0.000000" "0.000000,0.000000,0.000000" "Auto" false true) :!class /Script/FortniteGame.FortProjectileMovementComponent)
      (Variables|/Script/Engine.ProjectileMovementComponent|SetInitialSpeed (ScalableFloat|GetValueatLevel (Variables|Balance|GetChaseInitialSpeed)) _fortprojectilemovecomp_chase)
      (Variables|/Script/Engine.ProjectileMovementComponent|SetMaxSpeed (ScalableFloat|GetValueatLevel (Variables|Balance|GetChaseMaxSpeed)) _fortprojectilemovecomp_chase)
      (Components|Movement|SetUpdatedComponent _fortprojectilemovecomp_chase (Variables|Transformation|GetRootComponent))
      (Development|PrintString (Utilities|String|Append "New Homing Data set to target: " (Utilities|GetDisplayName (Targeting|GetChaseTarget)) "    Targeting Socket" (Utilities|String|ToString(Name) (Variables|Targeting|GetTargetSocketName))) (Variables|Parameters|GetDebug) (Variables|Parameters|GetDebug)))
    (:Is%20Not%20Valid
      (Development|PrintString "Homing Taret Invalid, going to patrol" (Variables|Parameters|GetDebug) (Variables|Parameters|GetDebug))
      (CallFunction|StartPatrolLoop))))

(fn ShouldAlliesAutoFireOnDeployedTarget
  (return false))

(fn ShouldAutoFireOnSelfDeployedTarget
  (return false))

(fn SwapHead
  (params (AttackTarget :type Actor%20Object%20Reference))
  (bind _faunagadgetlookup (Variables|Animation|GetFaunaGadgetLookup))
  (bind _playergadgetitem (Variables|Animation|GetPlayerGadgetItem))
  ;[1 1 1] Try to treat the attack target as a player
  (bind _asfort_player_pawn (Utilities|Casting|CastToFortPlayerPawn AttackTarget)
    (:then
      (if (Utilities|IsValid (Variables|Animation|GetPlayerGadgetItem) :!class /Script/Engine.KismetSystemLibrary)
        (if (Utilities|IsValid (Fortnite|GetFortPlayerControllerfromActor _asfort_player_pawn) :!class /Script/Engine.KismetSystemLibrary)
          (bind _asfort_inventory_owner_interface (Utilities|Casting|CastToFortInventoryOwnerInterface (Fortnite|GetFortPlayerControllerfromActor _asfort_player_pawn))
            (:then
              _)
            (:CastFailed
              (return)))
          (else
            (bind _asfort_athena_aibot_controller (Utilities|Casting|CastToFortAthenaAIBotController (Pawn|GetController _asfort_player_pawn))
              (:then
                (bind _asfort_inventory_owner_interface_2 (Utilities|Casting|CastToFortInventoryOwnerInterface _asfort_athena_aibot_controller)
                  (:then
                    _)
                  (:CastFailed
                    (return))))
              (:CastFailed
                (return)))))
        (Inventory|GiveItemtoInventoryOwner (select (Utilities|IsValid (Fortnite|GetFortPlayerControllerfromActor _asfort_player_pawn) :!class /Script/Engine.KismetSystemLibrary) _asfort_inventory_owner_interface _asfort_inventory_owner_interface_2) _playergadgetitem)
        (Variables|Animation|SetGadgetItemHeadSwap _playergadgetitem)
        (return)
        (else
          (return))))
    (:CastFailed
      ;[1 1 1] Treat the attack target as fauna
      (bind _asfort_aipawn (Utilities|Casting|CastToFortAIPawn AttackTarget)
        (:then
          (bind _keys (Utilities|Map|Keys _faunagadgetlookup))
          (bind (_array_element _array_index) (Utilities|Array|ForEachLoopwithBreak _keys :!label _label)
            (:LoopBody
              (if (GameplayTags|HasAllMatchingGameplayTags _asfort_aipawn (GameplayTags|MakeGameplayTagContainerfromTag _array_element))
                (bind _values (Utilities|Map|Values _faunagadgetlookup))
                (bind _result (ArrayMacros|SafeArrayGet _values _array_index)
                  (:%20%20
                    (bind _output_get (Variables|Animation|SetGadgetItemHeadSwap _result))
                    (Inventory|GiveItemtoInventoryOwner _asfort_aipawn _output_get)
                    (goto _label Break)))))
            (:Completed
              (return))))
        (:CastFailed
          (return))))))

(fn TogglePlayerHeadVFX
  (params (HideEffects :type Boolean) (Character :type Character%20Object%20Reference))
  (for _array_element (Utilities|Array|MakeArray "Head" "Hat" "Face")
    (bind _asfort_player_pawn (Utilities|Casting|CastToFortPlayerPawn Character))
    (Utilities|IsValid (Customization|GetSkeletalMeshforPartType _asfort_player_pawn _array_element :!class /Script/FortniteGame.FortPlayerPawn)
      (:Is%20Valid
        (Development|SetHiddeninGame (Customization|GetSkeletalMeshforPartType _asfort_player_pawn _array_element :!class /Script/FortniteGame.FortPlayerPawn) HideEffects true)))))

(fn UpdateVocalEffects
  (params (VocalSourceEffectChainOverride :type Sound%20Effect%20Source%20Preset%20Chain%20Object%20Reference))
  (LocalVariables|SetEffectChainLocal VocalSourceEffectChainOverride)
  (Fortnite|FortShippingLog self (Utilities|String|Append "AvacadoEater - Updating Vocal FX: Actor = " (Utilities|GetDisplayName (Variables|Behavior|GetAttachedCached)) "| Effect = " (Utilities|GetDisplayName (LocalVariables|GetEffectChainLocal))))
  (bind _asnpc_pawn_wildlife_parent (Utilities|Casting|CastToNPC_Pawn_Wildlife_Parent (Variables|Behavior|GetAttachedCached))
    (:then
      (Variables|/Game/Athena/AI/Pawns/NPC/Parents/NPC_Pawn_Wildlife_Parent.NPC_Pawn_Wildlife_Parent_C|SetVocalSourceEffectChainOverride (LocalVariables|GetEffectChainLocal) _asnpc_pawn_wildlife_parent))
    (:CastFailed
      (bind _asfort_player_pawn (Utilities|Casting|CastToFortPlayerPawn (Variables|Behavior|GetAttachedCached)))
      (if (Pawn|IsLocallyControlled _asfort_player_pawn)
        (Voice|SetVoiceInputEffectChain (Fortnite|GetFortPlayerControllerfromActor _asfort_player_pawn) (select (Utilities|IsValid (LocalVariables|GetEffectChainLocal) :!class /Script/Engine.KismetSystemLibrary) "/Game/Athena/Items/Consumables/AvacadoEater/DSPEffects/AvacadoEater_VOIP_EffectChain.AvacadoEater_VOIP_EffectChain" 0))
        (else
          (bind _asbp_player_pawn_non_participant (Utilities|Casting|CastToBP_PlayerPawn_NonParticipant _asfort_player_pawn))
          (Utilities|IsValid (Actor|GetComponentbyClass _asbp_player_pawn_non_participant "/Game/Sounds/Fort_Audio_Blueprints/VoiceComponent/NPC_VoiceComponent.NPC_VoiceComponent_C")
            (:Is%20Valid
              (Class|NPCVoiceComponent|SetSourceEffectChaintoApply (Actor|GetComponentbyClass _asbp_player_pawn_non_participant "/Game/Sounds/Fort_Audio_Blueprints/VoiceComponent/NPC_VoiceComponent.NPC_VoiceComponent_C") (LocalVariables|GetEffectChainLocal)))))))))

(fn ConstructionScript
  (bind _capsulecomponent (Variables|Collision|GetCapsuleComponent))
  (|CallParentFunction)
  (Components|Capsule|SetCapsuleHalfHeight _capsulecomponent (ScalableFloat|GetValueatLevel (Variables|Balance|GetCollisionHeight)) true)
  (Components|Capsule|SetCapsuleRadius _capsulecomponent (ScalableFloat|GetValueatLevel (Variables|Balance|GetCollisionRadius)) true))
