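; Parent Blueprint for the deployed Avacado Eater alien egg, a proximity trap that hatches a bird actor, optionally targeting whoever triggered it, or can be picked up.
; After an arming delay, overlapping players and NPCs start a looping warning, and a server timer tests the inner trigger sphere with tag and optional speed checks.
; Damage, hits, water landing or losing its attached base also hatch it.
; Hatching can chain-react through nearby eggs; eggs outside the safe zone expire.
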(event OnComponentBeginOverlap(WarningShape)
  (params (OverlappedComponent :type Primitive%20Component%20Object%20Reference) (OtherActor :type Actor%20Object%20Reference) (OtherComp :type Primitive%20Component%20Object%20Reference) (OtherBodyIndex :type Integer) (bFromSweep :type Boolean) (SweepResult :type Hit%20Result%20Structure%20%28by%20ref%29))
  ;[1 1 1] WARNING COLLISION SHAPE OVERLAP
  (if (Networking|IsServer)
    (if (Variables|States&Logic|GetHasFinishedDeploying?)
      (CallFunction|TestActortoTriggerWarning OtherActor))))

(event OnComponentEndOverlap(WarningShape)
  (params (OverlappedComponent :type Primitive%20Component%20Object%20Reference) (OtherActor :type Actor%20Object%20Reference) (OtherComp :type Primitive%20Component%20Object%20Reference) (OtherBodyIndex :type Integer))
  ;[1 1 1] WARNING COLLISION SHAPE OVERLAP
  (CallFunction|RemoveActorfromWarningTrack OtherActor))

(event Custom|Deploy
  ;[1 1 1] DEPLOY - called when the object is initially spawned
  (Variables|States&Logic|SetDeployParameters (Ability|GameplayCue|MakeGameplayCueParameters :NormalizedMagnitude 0.0 :RawMagnitude 0.0 :Location "0,0,0" :Normal "0,0,0" :GameplayEffectLevel (select (Variables|States&Logic|GetShouldGrowOnSpawn) 2 1) :AbilityLevel (select (Variables|States&Logic|GetShouldGrowOnSpawn) 2 1) :bReplicateLocationWhenUsingMinimalRepProxy false))
  (if (Variables|States&Logic|GetShouldGrowOnSpawn)
    (Transformation|SetRelativeScale3D (Variables|Default|GetStaticMesh) "0, 0, 0")
    (else
      (Ability|ExecuteGameplayCueWithParamsReplicated (Variables|GCs|GetGC_SpawnDeploy) (Variables|States&Logic|GetDeployParameters))))
  ;;[1 0.6298 0] Deploy feedbacks
  (EventDispatchers|BindEventtoOnDamaged (EventDispatchers|CreateEvent :!event Custom|DamageEvent) :!class /Script/FortniteGame.BuildingActor)
  (Fortnite|ShowSoundIndicatorLocal :TrackedActor self :Location (Transformation|GetActorLocation) :MaxAudibleDistance (Variables|Default|GetMax%20Audible%20Distance) :IndicatorType "Infected" :Instigator self :Tint "(R=0.000000,G=0.000000,B=0.000000,A=0.000000)" :OverrideIcon (Variables|Default|GetAudioVisual%20Indicator%20Icon) :AttenuationMaxRangeMultiplier 1.0 :AttenuationIntensityMultiplier 1.0)
  (Utilities|FlowControl|Delay 1.0)
  (if (Variables|States&Logic|GetShouldGrowOnSpawn)
    (Ability|ExecuteGameplayCueWithParamsReplicated (Variables|GCs|GetGC_SpawnDeploy) (Variables|States&Logic|GetDeployParameters))
    (bind (_direction _egggrowth) (Timeline|EggGrowth PlayFromStart :NewTime 1.0)
      (:Update
        (Transformation|SetRelativeScale3D (Variables|Default|GetStaticMesh) _egggrowth))
      (:Finished
        _)))
  (if (Networking|IsServer)
    (Utilities|FlowControl|Delay (ScalableFloat|GetValueatLevel (Variables|Balance|GetArmingDelay)))
    (Ability|ExecuteGameplayCueReplicated (Variables|GCs|GetGC_Armed))
    (Collision|SetCollisionEnabled (Variables|Default|GetWarningShape) "QueryOnly")
    (Collision|SetCollisionEnabled (Variables|Default|GetTriggerShape) "QueryOnly")
    (Variables|States&Logic|SetHasFinishedDeploying? true)
    (CallFunction|InitialDeploymentWarningCheck)))

(event EventBeginPlay
  ;[1 1 1] Begin Play
  (|CallParentFunction)
  (Class|FortAsyncActionCurrentPlaylistReady|PlaylistDataReadyAsync)
  (if (Networking|IsServer)
    (Utilities|FlowControl|Sequence
      (:then_0
        (Utilities|Array|Clear (Variables|States&Logic|GetActorsInWarningRange))
        (Variables|States&Logic|SetIsDying? false)
        (Variables|States&Logic|SetHatchHasTriggered? false)
        (Variables|States&Logic|SetHasFinishedDeploying? false)
        (Variables|States&Logic|SetHatchFinished? false)
        (Variables|States&Logic|SetAttachedBaseDestroyed? false)
        (Development|PrintString (Utilities|String|Append "Egg Instigator: " (Utilities|GetDisplayName (Variables|Actor|GetInstigator))) (Variables|Default|GetDebug) (Variables|Default|GetDebug)))
      (:then_1
        (CallFunction|SetupSafeZoneCheck)
        (CallFunction|Deploy))
      (:then_2
        ;;[1 1 1] Attempt to bind the freshly spawned egg to whatever its on
        (bind _outhit (Collision|SphereTraceForObjects :Start (+ (Transformation|GetActorLocation) (* (Transformation|GetActorUpVector) 1e+01)) :End (+ (Transformation|GetActorLocation) (* (Transformation|GetActorUpVector) -5e+01)) :Radius 15.0 :ObjectTypes (Utilities|Array|MakeArray "ObjectTypeQuery1" "ObjectTypeQuery5") :bTraceComplex true :DrawDebugType "None" :bIgnoreSelf true :TraceColor "(R=1.000000,G=0.000000,B=0.000000,A=1.000000)" :TraceHitColor "(R=0.000000,G=1.000000,B=0.000000,A=1.000000)" :DrawTime 5e+01))
        (if (or (.bSuccess (Utilities|Casting|CastToFortWaterBodyBP (.HitActor (Collision|BreakHitResult _outhit)))) (.bSuccess (Utilities|Casting|CastToFortAthenaWater (.HitActor (Collision|BreakHitResult _outhit)))))
          (CallFunction|StartHatchSequence 0 (Variables|Actor|GetInstigator))
          (elif (Variables|Parameters|GetAttachOnSpawn?)
            (CallFunction|BindtoSurfaceonDeploy _outhit)))))))

(event Custom|DamageEvent
  (params (DamagedActor :type Actor%20Object%20Reference) (Damage :type Float%20%28single-precision%29) (InstigatedBy :type Controller%20Object%20Reference) (DamageCauser :type Actor%20Object%20Reference) (HitLocation :type Vector) (FHitComponent :type Primitive%20Component%20Object%20Reference) (BoneName :type Name) (Momentum :type Vector))
  (bind _returnvalue (Pawn|GetControlledPawn InstigatedBy))
  ;[1 1 1] On Damage
  (if (Networking|IsServer)
    (Development|PrintString "Damaged" (Variables|Default|GetDebug) (Variables|Default|GetDebug))
    (Ability|ExecuteGameplayCueWithParamsReplicated (Variables|GCs|GetGC_Damage) (Ability|GameplayCue|MakeGameplayCueParameters :NormalizedMagnitude 0.0 :RawMagnitude 0.0 :Location HitLocation :Normal "0,0,0" :Instigator _returnvalue :EffectCauser _returnvalue :SourceObject _returnvalue :GameplayEffectLevel 1 :AbilityLevel 1 :bReplicateLocationWhenUsingMinimalRepProxy false))
    (CallFunction|StartHatchSequence _returnvalue _returnvalue)))

(event Custom|StopWarningState
  ;[1 1 1] WARNING STATE
  (Ability|RemoveGameplayCueReplicated (Variables|GCs|GetGC_Warning_loop))
  (CallFunction|StopWarningFeedback)
  (if (Utilities|Time|IsValidTimerHandle (stale Custom|StartWarningState _returnvalue))
    (Utilities|Time|ClearandInvalidateTimerbyHandle (stale Custom|StartWarningState _returnvalue))))

(event Custom|StopWarningFeedback
  ;[1 1 1] WARNING STATE
  ;;[1 0.6255 0] Looping AV Warning Feedback
  (if (not (Networking|IsServer))
    (Rendering|Material|SetCustomPrimitiveDataFloat (Variables|Default|GetStaticMesh) 0 0.0)
    (if (Utilities|Time|IsValidTimerHandle (stale Custom|StartWarningFeedback _returnvalue))
      (Utilities|Time|ClearandInvalidateTimerbyHandle (stale Custom|StartWarningFeedback _returnvalue)))))

(event Custom|StartWarningFeedback
  ;[1 1 1] WARNING STATE
  ;;[1 0.6255 0] Looping AV Warning Feedback
  (if (not (Networking|IsServer))
    (Rendering|Material|SetCustomPrimitiveDataFloat (Variables|Default|GetStaticMesh) 0 1.0)
    (bind _returnvalue (Utilities|Time|SetTimerbyEvent (delegate Custom|WarningAVIndicatorLoop) (Variables|Default|GetAudioVisualIndicator%20Update%20Rate) true))))

(event Custom|StartWarningState
  ;[1 1 1] WARNING STATE
  (Ability|AddGameplayCueReplicated (Variables|GCs|GetGC_Warning_loop))
  (CallFunction|StartWarningFeedback)
  (if (Networking|IsServer)
    (bind _returnvalue (Utilities|Time|SetTimerbyEvent (delegate Custom|WarningState) (ScalableFloat|GetValueatLevel (Variables|Balance|GetWarningUpdateRate)) true))))

(event Custom|WarningAVIndicatorLoop
  ;[1 1 1] WARNING STATE
  ;;[1 0.6255 0] Looping AV Warning Feedback
  (Fortnite|ShowSoundIndicatorLocal :TrackedActor self :Location (Transformation|GetActorLocation) :MaxAudibleDistance (ScalableFloat|GetValueatLevel (Variables|Balance|GetWarningRadius)) :IndicatorType "Infected" :Instigator self :Tint "(R=0.000000,G=0.000000,B=0.000000,A=0.000000)" :OverrideIcon (Variables|Default|GetAudioVisual%20Indicator%20Icon) :AttenuationMaxRangeMultiplier 1.0 :AttenuationIntensityMultiplier 1.0))

(event Custom|WarningState
  ;[1 1 1] WARNING STATE
  ;;[1 1 1] TRIGGER LOGIC: check if any valid object inside trigger radius AND is moving fast enough
  (for _array_element (Variables|States&Logic|GetActorsInWarningRange)
    (Utilities|IsValid _array_element
      (:Is%20Valid
        ;;;[1 1 1] Is the object inside the smaller "Trigger" radius?
        (if (Collision|IsOverlappingActor (Variables|Default|GetTriggerShape) _array_element :!class /Script/Engine.PrimitiveComponent)
          ;;;[1 1 1] Check for tags that should instantly trigger a tag like being in a vehicle
          (bind _returnvalue (GameplayTags|HasAnyMatchingGameplayTags(Message) _array_element (Variables|Parameters|GetActorHatchTriggeringTags)))
          (if _returnvalue
            _
            (elif (ScalableFloat|GetValueasBool (Variables|Balance|GetUseSpeedThreshold?))
              ;;;[1 1 1] Speed Threshold check ; if enabled the target must be in range AND going above the speed threshold to trigger hatching; allows sneaking past
              (if (>= (Math|Vector|VectorLength (Transformation|GetVelocity _array_element)) (ScalableFloat|GetValueatLevel (Variables|Balance|GetHatchSpeedThreshold)))
                _)))
          (CallFunction|StartHatchSequence _array_element (Variables|Actor|GetInstigator))))
      (:Is%20Not%20Valid
        (CallFunction|RemoveActorfromWarningTrack _array_element)))))

(event Custom|BindtoSurfaceOnDeploy
  (params (Hit%20Surface :type Hit%20Result%20Structure))
  (bind (_bblockinghit _binitialoverlap _time _distance _location _impactpoint _normal _impactnormal _physmat _hitactor _hitcomponent) (Collision|BreakHitResult Hit%20Surface))
  ;[1 1 1] Attach to surface on spawn
  (Utilities|IsValid _hitactor
    (:Is%20Valid
      (Transformation|SetActorLocation _impactpoint false true)
      (Transformation|SetActorRotation (Math|Rotator|MakeRotfromZ _impactnormal) true)
      (bind _asbuilding_actor (Utilities|Casting|CastToBuildingActor _hitactor)
        (:then
          (EventDispatchers|BindEventtoOnDied _asbuilding_actor (delegate Custom|BaseDied) :!class /Script/FortniteGame.BuildingActor)
          (CallFunction|AttachtoBindedActor _hitcomponent))
        (:CastFailed
          (Game|BindEventtoOnDestroyed _hitactor (delegate Custom|BaseDestroyed)))))))

(event Collision|EventHit
  (params (MyComp :type Primitive%20Component%20Object%20Reference) (Other :type Actor%20Object%20Reference) (OtherComp :type Primitive%20Component%20Object%20Reference) (bSelfMoved :type Boolean) (HitLocation :type Vector :default "0, 0, 0") (HitNormal :type Vector :default "0, 0, 0") (NormalImpulse :type Vector :default "0, 0, 0") (Hit :type Hit%20Result%20Structure%20%28by%20ref%29))
  ;[1 1 1] On hit: hatch!~
  (if (Networking|IsServer)
    (if (not (Variables|States&Logic|GetHatchHasTriggered?))
      (Utilities|Casting|CastToFortAthenaVehicle Other
        (:then
          (Development|PrintString "Vehicle triggered hatch" (Variables|Default|GetDebug) (Variables|Default|GetDebug)))
        (:CastFailed
          (bind _validtarget (CallFunction|ValidActor? Other))
          (if _validtarget
            (Development|PrintString "Something has hit the egg; triggering hatch" (Variables|Default|GetDebug) (Variables|Default|GetDebug)))))
      (CallFunction|StartHatchSequence Other Other))))

(event Custom|DestroyWithEffects
  ;[1 1 1] DEATH
  ;;[1 0.75 0] This is required to avoid running feedback and cleanup twice because this event is called both when HP reaches 0 AND when the actor is destroyed outright;
  (if (not (Variables|States&Logic|GetIsDying?))
    (Development|PrintString "Death" (Variables|Default|GetDebug) (Variables|Default|GetDebug))
    (Variables|States&Logic|SetIsDying? true)
    (Networking|ForceNetUpdate)))

(event Custom|BaseDied
  (params (DamagedActor :type Actor%20Object%20Reference) (Damage :type Float%20%28single-precision%29) (InstigatedBy :type Controller%20Object%20Reference) (DamageCauser :type Actor%20Object%20Reference) (HitLocation :type Vector) (FHitComponent :type Primitive%20Component%20Object%20Reference) (BoneName :type Name) (Momentum :type Vector))
  ;[1 1 1] Hatch Triggers when base is destroyed
  (CallFunction|StartHatchSequence DamageCauser DamageCauser)
  (Variables|States&Logic|SetAttachedBaseDestroyed? true)
  (Development|PrintString "Attached base destroyed, hatching" (Variables|Default|GetDebug) (Variables|Default|GetDebug)))

(event EventOnDeathServer
  (params (Damage :type Float%20%28single-precision%29) (DamageTags :type Gameplay%20Tag%20Container%20Structure%20%28by%20ref%29) (Momentum :type Vector :default "0, 0, 0") (HitInfo :type Hit%20Result%20Structure%20%28by%20ref%29) (InstigatedBy :type Controller%20Object%20Reference) (DamageCauser :type Actor%20Object%20Reference) (EffectContext :type Gameplay%20Effect%20Context%20Handle%20Structure))
  ;[1 1 1] DEATH
  ;;[1 0.75 0] This is required to avoid running feedback and cleanup twice because this event is called both when HP reaches 0 AND when the actor is destroyed outright;
  (if (not (Variables|States&Logic|GetIsDying?))
    (Development|PrintString "Death" (Variables|Default|GetDebug) (Variables|Default|GetDebug))
    (Variables|States&Logic|SetIsDying? true)
    (Networking|ForceNetUpdate)))

(event Custom|HatchFeedback
  ;[1 1 1] START HATCH SEQUENCE
  ;;[1 0.7362 0] Pop Membrane Feedback
  (Rendering|Material|SetCustomPrimitiveDataFloat (Variables|Default|GetStaticMesh) 1 1.0))

(event Custom|BaseDestroyed
  (params (DestroyedActor :type Actor%20Object%20Reference))
  ;[1 1 1] Hatch Triggers when base is destroyed
  (CallFunction|StartHatchSequence 0 (Variables|Actor|GetInstigator))
  (Variables|States&Logic|SetAttachedBaseDestroyed? true)
  (Development|PrintString "Attached base destroyed, hatching" (Variables|Default|GetDebug) (Variables|Default|GetDebug)))

(event Custom|ImmediatelyHide&Kill
  ;[1 1 1] Hide Actor and Kill - called by RepNotify on isDying
  (Collision|SetActorEnableCollision false)
  (Audio|Components|Audio|FadeOut (Variables|Default|GetAmbientSound) 0.25 0.0 "Linear" :!class /Script/Engine.AudioComponent)
  (Rendering|SetActorHiddenInGame true)
  (Actor|SetLifeSpan (Math|Random|RandomFloatinRange 0.9 1.0))
  (if (not (Variables|States&Logic|GetHatchFinished?))
    (Ability|ExecuteGameplayCueReplicated (Variables|GCs|GetGC_Destroy))))

(event Custom|StartHatchSequence
  (params (TargetedActor :type Actor%20Object%20Reference) (InstigatingActor :type Actor%20Object%20Reference) (CausedByChainReaction? :type Boolean))
  ;[1 1 1] START HATCH SEQUENCE
  (if (not (Variables|States&Logic|GetHatchHasTriggered?))
    (Variables|States&Logic|SetHatchHasTriggered? true)
    (Development|PrintString (Utilities|String|Append "HATCH triggered by:" (Utilities|GetDisplayName TargetedActor)) (Variables|Default|GetDebug) (Variables|Default|GetDebug))
    (Fortnite|ShowSoundIndicatorLocal :TrackedActor self :Location (Transformation|GetActorLocation) :MaxAudibleDistance (Variables|Default|GetMax%20Audible%20Distance) :IndicatorType "Infected" :Instigator self :Tint "(R=0.000000,G=0.000000,B=0.000000,A=0.000000)" :OverrideIcon (Variables|Default|GetAudioVisual%20Indicator%20Icon) :AttenuationMaxRangeMultiplier 1.0 :AttenuationIntensityMultiplier 1.0)
    (CallFunction|StopWarningState)
    (if (Networking|IsServer)
      (Ability|ExecuteGameplayCueReplicated (Variables|GCs|GetGC_HatchStarted))
      ;;[1 1 1] if enabled; Cancels hatching if avacado eater is dying
      (if (not (and (Variables|States&Logic|GetIsDying?) (ScalableFloat|GetValueasBool (Variables|Balance|GetDeathCancelsHatch?))))
        (CallFunction|HatchFeedback)
        (CallFunction|SpawnBGA TargetedActor InstigatingActor)
        (if (not CausedByChainReaction?)
          (bind _asfort_player_pawn (Utilities|Casting|CastToFortPlayerPawn InstigatingActor)
            (:then
              (Quest|SendComplexCustomStatEvent (Quests|GetQuestManager (Fortnite|GetFortPlayerControllerfromActor _asfort_player_pawn) "Invalid") 0 "(GameplayTags=)" "(GameplayTags=((TagName=\"Athena.Quests.AvacadoEater.Egg.Destroyed\")))" 1))
            (:CastFailed
              _))))
      (if (ScalableFloat|GetValueasBool (Variables|Balance|GetChainReactionEnabled?))
        (CallFunction|ChainReact TargetedActor InstigatingActor)
        (else
          (CallFunction|PostHatchDestroy))))))

(event Custom|SpawnBGA
  (params (TargetedActor :type Actor%20Object%20Reference) (InstigatingActor :type Actor%20Object%20Reference))
  (bind _bgatospawnonhatch (Variables|Parameters|GetBGAToSpawnOnHatch))
  ;[1 1 1] Spawn BGA
  (if (Utilities|IsValidClass _bgatospawnonhatch)
    (bind _returnvalue (Building|SpawnBuildingGameplayActor _bgatospawnonhatch (Math|Transform|MakeTransform (Transformation|GetSocketLocation (Variables|Default|GetStaticMesh) (Utilities|Array|Get(aref) (Variables|Parameters|GetSpawnPoints) (Math|Random|RandomInteger (Utilities|Array|Length (Variables|Parameters|GetSpawnPoints))))) (Transformation|GetActorRotation)) self))
    (bind _asbga_athena_avacado_eater_bird_parent (Utilities|Casting|CastToBGA_Athena_AvacadoEaterBird_Parent _returnvalue)
      (:then
        (Variables|/Game/Athena/Items/Consumables/AvacadoEater/BGA_Athena_AvacadoEaterBird_Parent.BGA_Athena_AvacadoEaterBird_Parent_C|SetDamageCauser (select (Utilities|IsValid InstigatingActor :!class /Script/Engine.KismetSystemLibrary) InstigatingActor self) _asbga_athena_avacado_eater_bird_parent)
        (if (and (ScalableFloat|GetValueasBool (Variables|Balance|GetHatchTargetInstigator?)) (Utilities|IsValid TargetedActor :!class /Script/Engine.KismetSystemLibrary))
          (Targeting|SetChaseTarget _asbga_athena_avacado_eater_bird_parent TargetedActor true 0.0)
          (Development|PrintString (Utilities|String|Append "SPAWNED BGA targeting:  " (Utilities|GetDisplayName TargetedActor)) (Variables|Default|GetDebug) (Variables|Default|GetDebug))
          (else
            (Development|PrintString "SPAWNED  BGA with NO Target" (Variables|Default|GetDebug) (Variables|Default|GetDebug) :!label _label)))
        (Ability|ExecuteGameplayCueReplicated (Variables|GCs|GetGC_HatchSpawned))
        (CallFunction|HatchAVIndicator)
        (Variables|States&Logic|SetHatchFinished? true))
      (:CastFailed
        _))
    (goto _label)))

(event Custom|SetupSafeZoneCheck
  ;[1 1 1] SAFE ZONE CHECK
  (bind _returnvalue (Utilities|Time|SetTimerbyEvent (delegate Custom|SafeZoneCheck) (ScalableFloat|GetValueatLevel (Variables|Balance|GetSafeZoneCheckRate)) true)))

(event Custom|SafeZoneCheck
  ;[1 1 1] SAFE ZONE CHECK
  (bind _issafe (CallFunction|SafeZoneStatusCheck))
  (if (not _issafe)
    (Utilities|FlowControl|Delay (ScalableFloat|GetValueatLevel (Variables|Balance|GetLifetimeOutsideSafeZone)))
    (CallFunction|DestroywithEffects)
    (if (Utilities|Time|IsValidTimerHandle (stale Custom|SetupSafeZoneCheck _returnvalue))
      (Utilities|Time|ClearandInvalidateTimerbyHandle (stale Custom|SetupSafeZoneCheck _returnvalue)))))

(event Custom|HatchAVIndicator
  ;[1 1 1] Spawn BGA
  ;;[1 0.7734 0] AV Indicator
  (Fortnite|ShowSoundIndicatorLocal :TrackedActor self :Location (Transformation|GetActorLocation) :MaxAudibleDistance (Variables|Default|GetMax%20Audible%20Distance) :IndicatorType "Infected" :Instigator self :Tint "(R=0.000000,G=0.000000,B=0.000000,A=0.000000)" :OverrideIcon (Variables|Default|GetAudioVisual%20Indicator%20Icon) :AttenuationMaxRangeMultiplier 1.0 :AttenuationIntensityMultiplier 1.0))

(event Custom|ChainReact
  (params (TargetedActor :type Actor%20Object%20Reference) (InstigatingActor :type Actor%20Object%20Reference))
  ;[1 1 1] CHAIN REACTION
  ;;[1 0.7478 0.1881] Find nearby Avacado Eaters to hatch
  (bind _outactors (Collision|SphereOverlapActors (Transformation|GetActorLocation) (ScalableFloat|GetValueatLevel (Variables|Balance|GetChainReactionRange)) (Utilities|Array|MakeArray "ObjectTypeQuery1" "ObjectTypeQuery2" "ObjectTypeQuery6") "/Game/Athena/Items/Consumables/AvacadoEater/BGA_Athena_AvacadoEater_Deployed_Parent.BGA_Athena_AvacadoEater_Deployed_Parent_C" (Utilities|Array|MakeArray self)))
  (Variables|States&Logic|SetChainHatchTargets _outactors)
  (CallFunction|ChainLoopIteration))

(event Custom|ChainLoopIteration
  ;[1 1 1] CHAIN REACTION
  (Utilities|Time|SetTimerbyEvent (delegate Custom|hatch_opening_delay) (ScalableFloat|GetValueatLevel (Variables|Balance|GetHatchSpawnDelay))))

(event Interaction|EventBlueprintOnInteract
  (params (InteractingPawn :type Fort%20Pawn%20Object%20Reference) (InteractionBeingAttempted :type EInteractionBeingAttempted%20Enum :default "FirstInteraction"))
  (bind _itemtopickupondisarm (Variables|Parameters|GetItemToPickUpOnDisarm))
  ;[1 1 1] Pickup Interaction
  (if (not (or (Variables|States&Logic|GetHatchHasTriggered?) (Variables|States&Logic|GetPickupTriggered?)))
    (bind _asfort_player_controller (Utilities|Casting|CastToFortPlayerController (Pawn|GetController InteractingPawn)))
    (Utilities|IsValid _itemtopickupondisarm
      (:Is%20Valid
        (Inventory|GiveItemtoInventoryOwner _asfort_player_controller _itemtopickupondisarm)
        (Variables|States&Logic|SetPickupTriggered? true)
        (Ability|ExecuteGameplayCueReplicated (Variables|GCs|GetGC_Pickup))
        (CallFunction|DestroywithEffects)))))

(event Custom|hatch_opening_delay
  (bind _chainhatchtargets (Variables|States&Logic|GetChainHatchTargets))
  (bind _returnvalue (Utilities|Array|Length _chainhatchtargets))
  (bind _returnvalue_2 (- _returnvalue 1))
  (bind _output (Utilities|Array|Get(acopy) _chainhatchtargets _returnvalue_2))
  ;[1 1 1] CHAIN REACTION
  (if (> (Utilities|Array|Length (Variables|States&Logic|GetChainHatchTargets)) 0)
    (Utilities|IsValid _output
      (:Is%20Valid
        (bind _asbga_athena_avacado_eater_deployed_parent (Utilities|Casting|CastToBGA_Athena_AvacadoEater_Deployed_Parent _output)
          (:then
            (CallFunction|StartHatchSequence _asbga_athena_avacado_eater_deployed_parent (select (Math|Boolean|ANDBoolean (Utilities|IsValid (stale Custom|ChainReact TargetedActor) :!class /Script/Engine.KismetSystemLibrary) (ScalableFloat|GetValueasBool (Variables|Balance|GetChainHatchTargetsInstigator?)) (ScalableFloat|GetValueasBool (Variables|Balance|GetHatchTargetInstigator?))) (stale Custom|ChainReact TargetedActor) 0) (stale Custom|ChainReact InstigatingActor) true))
          (:CastFailed
            _)))
      (:Is%20Not%20Valid
        _))
    (Utilities|Array|RemoveIndex (Variables|States&Logic|GetChainHatchTargets) _returnvalue_2)
    (Utilities|Time|SetTimerbyEvent (delegate Custom|chaindlay) (Math|Random|RandomFloatinRange 0.1 0.2))
    (else
      (CallFunction|PostHatchDestroy))))

(event Custom|chaindlay
  ;[1 1 1] CHAIN REACTION
  (CallFunction|ChainLoopIteration))

(event Custom|AttachedWasDestroyed_1
  ;[1 1 1] Attached BGA Base is destroyed; freefall + hatch immediately
  (Development|PrintString "Base Destroyed")
  (CallFunction|StartHatchSequence 0 (Variables|Actor|GetInstigator)))

(event Custom|PostHatchDestroy
  ;[1 1 1] Shrink & Destroy
  (if (Variables|States&Logic|GetAttachedBaseDestroyed?)
    (if (Networking|IsServer)
      (CallFunction|DestroywithEffects))
    (else
      (CallFunction|ShrinkRoots))))

(event Custom|ShrinkRoots
  ;[1 1 1] Shrink & Destroy
  ;;[1 0.746 0.1154] Shrink the roots away
  (Rendering|Material|SetCustomPrimitiveDataFloat (Variables|Default|GetStaticMesh) 2 (Utilities|Time|GetGameTimeinSeconds))
  (Collision|SetActorEnableCollision false)
  (Audio|Components|Audio|FadeOut (Variables|Default|GetAmbientSound) 0.25 0.0 "Linear" :!class /Script/Engine.AudioComponent)
  (Utilities|Time|SetTimerbyEvent (delegate Custom|ShrinkDlay) (ScalableFloat|GetValueatLevel (Variables|Balance|GetPost%20Hatch%20Destroy%20Delay))))

(event Custom|ShrinkDlay
  ;[1 1 1] Shrink & Destroy
  (if (Networking|IsServer)
    (CallFunction|DestroywithEffects)))

(fn AttachToBindedActor
  (params (Attach%20Comp :type Primitive%20Component%20Object%20Reference))
  (bind _returnvalue (Components|GetOwner Attach%20Comp))
  (Utilities|IsValid Attach%20Comp
    (:Is%20Valid
      (Utilities|Casting|CastToFortVehicleInterface _returnvalue
        (:then
          _)
        (:CastFailed
          (Utilities|Casting|CastToBuildingActor _returnvalue
            (:then
              _)
            (:CastFailed
              (Utilities|Casting|CastToBuildingWall _returnvalue)))))
      (Transformation|AttachActorToComponent Attach%20Comp "None" "KeepWorld" "KeepWorld" "KeepWorld")
      ;[1 0 0.3772] what is this?
      (Collision|IgnoreActorwhenMoving (Variables|Default|GetStaticMesh) _returnvalue true))))

(fn BlueprintCanInteract
  (params (InteractingPawn :type Fort%20Pawn%20Object%20Reference) (InteractionBeingAttempted :type EInteractionBeingAttempted%20Enum :default "FirstInteraction") (InteractionType :type TInteractionType%20Enum :default "IT_NoInteraction"))
  (return (and (not (Variables|States&Logic|GetHatchHasTriggered?)) (ScalableFloat|GetValueasBool (Variables|Balance|GetCanBePickedUp?)))))

(fn BlueprintGetInteractionString
  (params (InteractingPawn :type Fort%20Pawn%20Object%20Reference) (InteractionBeingAttempted :type EInteractionBeingAttempted%20Enum :default "FirstInteraction"))
  (return "NSLOCTEXT(\"[2749D6204A911DC06983168F53801394]\", \"AD7BB8414AEF5F6DDB64119B085BDA57\", \"Pickup Alien Egg\")"))

(fn GetTargetOwner
  (return 0))

(fn InitialDeploymentWarningCheck
  (for _array_element (Collision|GetOverlappingActors (Variables|Default|GetWarningShape) 0 :!class /Script/Engine.PrimitiveComponent)
    (CallFunction|TestActortoTriggerWarning _array_element)))

(fn IsValidAutoFireTarget
  (params (Attacker :type Actor%20Object%20Reference))
  (return true))

(fn OnRep_IsDying?
  (if (Variables|States&Logic|GetIsDying?)
    (if (Networking|IsServer)
      (CallFunction|ImmediatelyHide&Kill))))

(fn RemoveActorFromWarningTrack
  (params (Actor :type Actor%20Object%20Reference))
  (bind _actorsinwarningrange (Variables|States&Logic|GetActorsInWarningRange))
  (if (Networking|IsServer)
    (Utilities|Array|RemoveItem _actorsinwarningrange Actor)
    (if (<= (Utilities|Array|Length _actorsinwarningrange) 0)
      (CallFunction|StopWarningState))))

(fn SafeZoneStatusCheck
  (bind _returnvalue (SafeZone|GetSafeZoneIndicator))
  (LocalVariables|Setsafe? true)
  (Utilities|IsValid _returnvalue
    (:Is%20Valid
      (if (not (< (Math|Vector|Distance(Vector) (Transformation|GetActorLocation) (SafeZone|GetSafeZoneCenter _returnvalue)) (SafeZone|GetSafeZoneRadius _returnvalue)))
        (LocalVariables|Setsafe? false)))
    (:Is%20Not%20Valid
      _))
  (return (LocalVariables|Getsafe?)))

(fn ShouldAlliesAutoFireOnDeployedTarget
  (return false))

(fn ShouldAutoFireOnSelfDeployedTarget
  (return false))

(fn TestActorToTriggerWarning
  (params (TriggeringActor :type Actor%20Object%20Reference))
  (bind _actorsinwarningrange (Variables|States&Logic|GetActorsInWarningRange))
  (bind _validtarget (CallFunction|ValidActor? TriggeringActor))
  (if _validtarget
    ;[1 1 1] if we have > 1 actor in warning range we can start the warning state
    (Utilities|Array|AddUnique _actorsinwarningrange TriggeringActor)
    (if (> (Utilities|Array|Length _actorsinwarningrange) 0)
      (CallFunction|StartWarningState))))

(fn ConstructionScript
  (|CallParentFunction)
  (bind _returnvalue (Rendering|Material|CreateDynamicMaterialInstance (Variables|Default|GetStaticMesh) 0 0 "None" :!class /Script/Engine.PrimitiveComponent))
  (Variables|States&Logic|SetEggMaterial _returnvalue)
  (Collision|SetCollisionEnabled (Variables|Default|GetWarningShape) "NoCollision")
  (Collision|SetCollisionEnabled (Variables|Default|GetTriggerShape) "NoCollision")
  (Components|Sphere|SetSphereRadius (Variables|Default|GetWarningShape) (ScalableFloat|GetValueatLevel (Variables|Balance|GetWarningRadius)) true) ; NOTE: scale of root is set to 2 when BGA is spawned, so this radius is actually doubled.
  (Components|Sphere|SetSphereRadius (Variables|Default|GetTriggerShape) (ScalableFloat|GetValueatLevel (Variables|Balance|GetTriggerRadius)) true) ; NOTE: scale of root is set to 2 when BGA is spawned, so this radius is actually doubled.
)

(fn ValidActor?
  (params (ActorToTest :type Actor%20Object%20Reference))
  (LocalVariables|SetisValid false)
  ;[1 1 1] MUST have any of these tags to trigger hatch
  (bind _returnvalue (GameplayTags|HasAnyMatchingGameplayTags(Message) ActorToTest (GameplayTags|MakeGameplayTagContainerfromArray (Utilities|Array|MakeArray "(TagName=\"Pawn.Athena.NPC\")" "(TagName=\"Pawn.Athena.Player\")"))))
  (if _returnvalue
    ;[1 1 1] Must NOT have any of these tags
    (bind _returnvalue_2 (GameplayTags|HasAnyMatchingGameplayTags(Message) ActorToTest (Variables|Parameters|GetTags_IgnoredByAvacadoEater)))
    (if (not _returnvalue_2)
      (LocalVariables|SetisValid true)))
  (return (LocalVariables|GetisValid)))
